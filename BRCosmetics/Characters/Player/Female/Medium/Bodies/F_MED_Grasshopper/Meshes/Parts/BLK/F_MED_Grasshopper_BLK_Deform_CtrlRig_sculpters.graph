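# Corrective deformation Control Rig for the female medium Grasshopper body (BLK part), posing helper bones as limbs bend.
# Euler Z angles between limb bones are remapped to clamped weights that slide the gauntlets and drive elbow, knee, pectoral, deltoid and thigh bones.
# Weighted local overrides set translation, scale and initial rotation plus an offset.
# Elbows and knees layer two angle ranges; shoulders and thighs use absolute angles, tuned separately per side.
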ControlRigBlueprint /BRCosmetics/Characters/Player/Female/Medium/Bodies/F_MED_Grasshopper/Meshes/Parts/BLK/F_MED_Grasshopper_BLK_Deform_CtrlRig_sculpters
graphs [
 RigVMFunctionLibrary {links [], nodes []}
 RigVMModel {nodes [
   RigUnit_BeginExecution = &RigUnit_BeginExecution()
   n1 = &RigUnit_GetRelativeTransformForItem(Child=(Type=Bone,Name="upperarm_l"), Parent=(Type=Bone,Name="lowerarm_l"))
   &RigUnit_ModifyTransforms(
    ItemToModify=((Item=(Type=Bone,Name="dyn_l_gauntlet"),Transform=(Rotation=(X=0,Y=0,Z=0,W=1),Translation=(X=2,Y=0,Z=0),Scale3D=(X=1,Y=1,Z=1))))
    Weight=&RigVMFunction_MathFloatRemap(
     SourceMaximum=-87
     SourceMinimum=-18
     TargetMinimum=0
     bClamp=true
     Value=&RigVMFunction_MathQuaternionToEuler(RotationOrder=ZYX, Value=@n1.RelativeTransform.Rotation).Result.Z
    ).Result
    ExecutePin=@RigUnit_BeginExecution.ExecutePin
   )
   n2 = &RigUnit_GetRelativeTransformForItem(Child=(Type=Bone,Name="upperarm_r"), Parent=(Type=Bone,Name="lowerarm_r"))
   > &RigUnit_ModifyTransforms(
    ItemToModify=((Item=(Type=Bone,Name="dyn_r_gauntlet"),Transform=(Rotation=(X=0,Y=0,Z=0,W=1),Translation=(X=-4.5,Y=0,Z=0),Scale3D=(X=1,Y=1,Z=1))))
    Weight=&RigVMFunction_MathFloatRemap(
     SourceMaximum=-114
     SourceMinimum=-18
     TargetMinimum=0
     bClamp=true
     Value=&RigVMFunction_MathQuaternionToEuler(RotationOrder=ZYX, Value=@n2.RelativeTransform.Rotation).Result.Z
    ).Result
   )
   RigUnit_GetRelativeBoneTransform = &RigUnit_GetRelativeBoneTransform(Bone=lowerarm_l, Space=upperarm_l)
   n3 = &RigVMFunction_MathQuaternionToEuler(RotationOrder=ZYX, Value=@RigUnit_GetRelativeBoneTransform.Transform.Rotation)
   n4 = &RigVMFunction_MathFloatRemap(SourceMaximum=92, SourceMinimum=28, TargetMinimum=0, bClamp=true, Value=@n3.Result.Z)
   n5 = &RigVMFunction_MathQuaternionToEuler(
    RotationOrder=ZYX
    Value=/Script/ControlRig.RigUnit_GetInitialBoneTransform(Bone=deform_elbow_bk_l, Space=LocalSpace) {is=struct RigVMUnitNode}.Transform.Rotation
   )
   > /Script/ControlRig.RigUnit_ModifyBoneTransforms(
    BoneToModify=((Bone="deform_elbow_bk_l",Transform=(Rotation=(X=0,Y=0,Z=0,W=1),Translation=(X=-3.505,Y=-1.083,Z=0),Scale3D=(X=1,Y=1,Z=1))))
    Mode=OverrideLocal
    WeightMaximum=1
    Weight=@n4.Result
    BoneToModify.0.Transform.Rotation=&RigVMFunction_MathQuaternionFromEuler(Euler=/Script/RigVM.RigVMFunction_MathVectorAdd(B=(X=0,Y=0,Z=0), A=@n5.Result) {is=struct RigVMUnitNode, template="Add::Execute(in A,in B,out Result)"}.Result).Result
   ) {is=struct RigVMUnitNode}
   n6 = &RigVMFunction_MathQuaternionToEuler(RotationOrder=ZYX, Value=&RigUnit_GetInitialBoneTransform(Bone=deform_elbow_fr_l, Space=LocalSpace).Transform.Rotation)
   > &RigUnit_ModifyBoneTransforms(
    BoneToModify=((Bone="deform_elbow_fr_l",Transform=(Rotation=(X=0,Y=0,Z=0,W=1),Translation=(X=20.65,Y=2.176,Z=0),Scale3D=(X=1,Y=1,Z=1.125))))
    Mode=OverrideLocal
    WeightMaximum=1
    WeightMinimum=0
    Weight=@n4.Result
    BoneToModify.0.Transform.Rotation=&RigVMFunction_MathQuaternionFromEuler(Euler=&RigVMFunction_MathVectorAdd(B=(X=0,Y=0,Z=-42.186), A=@n6.Result) {template="Add::Execute(in A,in B,out Result)"}.Result).Result
   )
   n7 = &RigVMFunction_MathFloatRemap(SourceMaximum=148, SourceMinimum=91, TargetMinimum=0, bClamp=true, Value=@n3.Result.Z)
   > &RigUnit_ModifyBoneTransforms(
    BoneToModify=((Bone="deform_elbow_bk_l",Transform=(Rotation=(X=0,Y=0,Z=0,W=1),Translation=(X=-5,Y=-1,Z=0),Scale3D=(X=1,Y=1,Z=1))))
    Mode=OverrideLocal
    WeightMaximum=1
    Weight=@n7.Result
    BoneToModify.0.Transform.Rotation=&RigVMFunction_MathQuaternionFromEuler(Euler=&RigVMFunction_MathVectorAdd(B=(X=0,Y=0,Z=0), A=@n5.Result) {template="Add::Execute(in A,in B,out Result)"}.Result).Result
   )
   > &RigUnit_ModifyBoneTransforms(
    BoneToModify=((Bone="deform_elbow_fr_l",Transform=(Rotation=(X=0,Y=0,Z=0,W=1),Translation=(X=15,Y=2.176,Z=0),Scale3D=(X=1,Y=1,Z=1.337))))
    Mode=OverrideLocal
    WeightMaximum=1
    WeightMinimum=0
    Weight=@n7.Result
    BoneToModify.0.Transform.Rotation=&RigVMFunction_MathQuaternionFromEuler(Euler=&RigVMFunction_MathVectorAdd(B=(X=0,Y=0,Z=-50), A=@n6.Result) {template="Add::Execute(in A,in B,out Result)"}.Result).Result
   )
   n8 = &RigVMFunction_MathQuaternionToEuler(RotationOrder=ZYX, Value=&RigUnit_GetRelativeBoneTransform(Bone=lowerarm_r, Space=upperarm_r).Transform.Rotation)
   n9 = &RigVMFunction_MathFloatRemap(SourceMaximum=92, SourceMinimum=28, TargetMinimum=0, bClamp=true, Value=@n8.Result.Z)
   n10 = &RigVMFunction_MathQuaternionToEuler(RotationOrder=ZYX, Value=&RigUnit_GetInitialBoneTransform(Bone=deform_elbow_bk_r, Space=LocalSpace).Transform.Rotation)
   > &RigUnit_ModifyBoneTransforms(
    BoneToModify=((Bone="deform_elbow_bk_r",Transform=(Rotation=(X=0,Y=0,Z=0,W=1),Translation=(X=3.505,Y=1.083,Z=0),Scale3D=(X=1,Y=1,Z=1))))
    Mode=OverrideLocal
    WeightMaximum=1
    Weight=@n9.Result
    BoneToModify.0.Transform.Rotation=&RigVMFunction_MathQuaternionFromEuler(Euler=&RigVMFunction_MathVectorAdd(B=(X=0,Y=0,Z=0), A=@n10.Result) {template="Add::Execute(in A,in B,out Result)"}.Result).Result
   )
   n11 = &RigVMFunction_MathQuaternionToEuler(RotationOrder=ZYX, Value=&RigUnit_GetInitialBoneTransform(Bone=deform_elbow_fr_r, Space=LocalSpace).Transform.Rotation)
   > &RigUnit_ModifyBoneTransforms(
    BoneToModify=((Bone="deform_elbow_fr_r",Transform=(Rotation=(X=0,Y=0,Z=0,W=1),Translation=(X=-20.65,Y=-2.176,Z=0),Scale3D=(X=1,Y=1,Z=1.125))))
    Mode=OverrideLocal
    WeightMaximum=1
    WeightMinimum=0
    Weight=@n9.Result
    BoneToModify.0.Transform.Rotation=&RigVMFunction_MathQuaternionFromEuler(Euler=&RigVMFunction_MathVectorAdd(B=(X=0,Y=0,Z=-42.186), A=@n11.Result) {template="Add::Execute(in A,in B,out Result)"}.Result).Result
   )
   n12 = &RigVMFunction_MathFloatRemap(SourceMaximum=148, SourceMinimum=91, TargetMinimum=0, bClamp=true, Value=@n8.Result.Z)
   > &RigUnit_ModifyBoneTransforms(
    BoneToModify=((Bone="deform_elbow_bk_r",Transform=(Rotation=(X=0,Y=0,Z=0,W=1),Translation=(X=5,Y=1,Z=0),Scale3D=(X=1,Y=1,Z=1))))
    Mode=OverrideLocal
    WeightMaximum=1
    Weight=@n12.Result
    BoneToModify.0.Transform.Rotation=&RigVMFunction_MathQuaternionFromEuler(Euler=&RigVMFunction_MathVectorAdd(B=(X=0,Y=0,Z=0), A=@n10.Result) {template="Add::Execute(in A,in B,out Result)"}.Result).Result
   )
   > &RigUnit_ModifyBoneTransforms(
    BoneToModify=((Bone="deform_elbow_fr_r",Transform=(Rotation=(X=0,Y=0,Z=0,W=1),Translation=(X=-15,Y=-2.176,Z=0),Scale3D=(X=1,Y=1,Z=1.337))))
    Mode=OverrideLocal
    WeightMaximum=1
    WeightMinimum=0
    Weight=@n12.Result
    BoneToModify.0.Transform.Rotation=&RigVMFunction_MathQuaternionFromEuler(Euler=&RigVMFunction_MathVectorAdd(B=(X=0,Y=0,Z=-50), A=@n11.Result) {template="Add::Execute(in A,in B,out Result)"}.Result).Result
   )
   n13 = &RigVMFunction_MathQuaternionToEuler(RotationOrder=ZYX, Value=&RigUnit_GetRelativeBoneTransform(Bone=calf_l, Space=thigh_l).Transform.Rotation)
   n14 = &RigVMFunction_MathFloatRemap(SourceMaximum=90, SourceMinimum=4, TargetMinimum=0, bClamp=true, Value=@n13.Result.Z)
   n15 = &RigVMFunction_MathQuaternionToEuler(RotationOrder=ZYX, Value=&RigUnit_GetInitialBoneTransform(Bone=deform_knee_bk_01_l, Space=LocalSpace).Transform.Rotation)
   > &RigUnit_ModifyBoneTransforms(
    BoneToModify=((Bone="deform_knee_bk_01_l",Transform=(Rotation=(X=0,Y=0,Z=0,W=1),Translation=(X=-9.801,Y=-12.293,Z=0),Scale3D=(X=1,Y=1,Z=1.5))))
    Mode=OverrideLocal
    WeightMaximum=1
    Weight=@n14.Result
    BoneToModify.0.Transform.Rotation=&RigVMFunction_MathQuaternionFromEuler(Euler=&RigVMFunction_MathVectorAdd(B=(X=0,Y=0,Z=3.141), A=@n15.Result) {template="Add::Execute(in A,in B,out Result)"}.Result).Result
   )
   n16 = &RigVMFunction_MathQuaternionToEuler(RotationOrder=ZYX, Value=&RigUnit_GetInitialBoneTransform(Bone=deform_knee_fr_01_l, Space=LocalSpace).Transform.Rotation)
   > &RigUnit_ModifyBoneTransforms(
    BoneToModify=((Bone="deform_knee_fr_01_l",Transform=(Rotation=(X=0,Y=0,Z=0,W=1),Translation=(X=1.196,Y=1.065,Z=0),Scale3D=(X=1,Y=1,Z=1))))
    Mode=OverrideLocal
    WeightMaximum=1
    WeightMinimum=0
    Weight=@n14.Result
    BoneToModify.0.Transform.Rotation=&RigVMFunction_MathQuaternionFromEuler(Euler=&RigVMFunction_MathVectorAdd(B=(X=0,Y=0,Z=0), A=@n16.Result) {template="Add::Execute(in A,in B,out Result)"}.Result).Result
   )
   n17 = &RigVMFunction_MathFloatRemap(SourceMaximum=130, SourceMinimum=89, TargetMinimum=0, bClamp=true, Value=@n13.Result.Z)
   > &RigUnit_ModifyBoneTransforms(
    BoneToModify=((Bone="deform_knee_bk_01_l",Transform=(Rotation=(X=0,Y=0,Z=0,W=1),Translation=(X=-11,Y=-14,Z=0),Scale3D=(X=1,Y=1,Z=1.6))))
    Mode=OverrideLocal
    WeightMaximum=1
    WeightMinimum=0
    Weight=@n17.Result
    BoneToModify.0.Transform.Rotation=&RigVMFunction_MathQuaternionFromEuler(Euler=&RigVMFunction_MathVectorAdd(B=(X=0,Y=0,Z=5), A=@n15.Result) {template="Add::Execute(in A,in B,out Result)"}.Result).Result
   )
   > &RigUnit_ModifyBoneTransforms(
    BoneToModify=((Bone="deform_knee_fr_01_l",Transform=(Rotation=(X=0,Y=0,Z=0,W=1),Translation=(X=1.3,Y=1.5,Z=0),Scale3D=(X=1,Y=1,Z=1))))
    Mode=OverrideLocal
    WeightMaximum=1
    WeightMinimum=0
    Weight=@n17.Result
    BoneToModify.0.Transform.Rotation=&RigVMFunction_MathQuaternionFromEuler(Euler=&RigVMFunction_MathVectorAdd(B=(X=0,Y=0,Z=0), A=@n16.Result) {template="Add::Execute(in A,in B,out Result)"}.Result).Result
   )
   n18 = &RigVMFunction_MathQuaternionToEuler(RotationOrder=ZYX, Value=&RigUnit_GetRelativeBoneTransform(Bone=calf_r, Space=thigh_r).Transform.Rotation)
   n19 = &RigVMFunction_MathFloatRemap(SourceMaximum=90, SourceMinimum=4, TargetMinimum=0, bClamp=true, Value=@n18.Result.Z)
   n20 = &RigVMFunction_MathQuaternionToEuler(RotationOrder=ZYX, Value=&RigUnit_GetInitialBoneTransform(Bone=deform_knee_bk_01_r, Space=LocalSpace).Transform.Rotation)
   > &RigUnit_ModifyBoneTransforms(
    BoneToModify=((Bone="deform_knee_bk_01_r",Transform=(Rotation=(X=0,Y=0,Z=0,W=1),Translation=(X=9.801,Y=12.293,Z=0),Scale3D=(X=1,Y=1,Z=1.5))))
    Mode=OverrideLocal
    WeightMaximum=1
    WeightMinimum=0
    Weight=@n19.Result
    BoneToModify.0.Transform.Rotation=&RigVMFunction_MathQuaternionFromEuler(Euler=&RigVMFunction_MathVectorAdd(B=(X=0,Y=0,Z=3.141), A=@n20.Result) {template="Add::Execute(in A,in B,out Result)"}.Result).Result
   )
   n21 = &RigVMFunction_MathQuaternionToEuler(RotationOrder=ZYX, Value=&RigUnit_GetInitialBoneTransform(Bone=deform_knee_fr_01_r, Space=LocalSpace).Transform.Rotation)
   > &RigUnit_ModifyBoneTransforms(
    BoneToModify=((Bone="deform_knee_fr_01_r",Transform=(Rotation=(X=0,Y=0,Z=0,W=1),Translation=(X=-1.196,Y=-1.065,Z=0),Scale3D=(X=1,Y=1,Z=1))))
    Mode=OverrideLocal
    WeightMaximum=1
    WeightMinimum=0
    Weight=@n19.Result
    BoneToModify.0.Transform.Rotation=&RigVMFunction_MathQuaternionFromEuler(Euler=&RigVMFunction_MathVectorAdd(B=(X=0,Y=0,Z=0), A=@n21.Result) {template="Add::Execute(in A,in B,out Result)"}.Result).Result
   )
   n22 = &RigVMFunction_MathFloatRemap(SourceMaximum=130, SourceMinimum=89, TargetMinimum=0, bClamp=true, Value=@n18.Result.Z)
   > &RigUnit_ModifyBoneTransforms(
    BoneToModify=((Bone="deform_knee_bk_01_r",Transform=(Rotation=(X=0,Y=0,Z=0,W=1),Translation=(X=11,Y=14,Z=0),Scale3D=(X=1,Y=1,Z=1.6))))
    Mode=OverrideLocal
    WeightMaximum=1
    WeightMinimum=0
    Weight=@n22.Result
    BoneToModify.0.Transform.Rotation=&RigVMFunction_MathQuaternionFromEuler(Euler=&RigVMFunction_MathVectorAdd(B=(X=0,Y=0,Z=5), A=@n20.Result) {template="Add::Execute(in A,in B,out Result)"}.Result).Result
   )
   > &RigUnit_ModifyBoneTransforms(
    BoneToModify=((Bone="deform_knee_fr_01_r",Transform=(Rotation=(X=0,Y=0,Z=0,W=1),Translation=(X=-1.3,Y=-1.5,Z=0),Scale3D=(X=1,Y=1,Z=1))))
    Mode=OverrideLocal
    WeightMaximum=1
    WeightMinimum=0
    Weight=@n22.Result
    BoneToModify.0.Transform.Rotation=&RigVMFunction_MathQuaternionFromEuler(Euler=&RigVMFunction_MathVectorAdd(B=(X=0,Y=0,Z=0), A=@n21.Result) {template="Add::Execute(in A,in B,out Result)"}.Result).Result
   )
   n23 = &RigUnit_GetRelativeBoneTransform(Bone=upperarm_l, Space=spine_05)
   n24 = &RigVMFunction_MathFloatRemap(
    SourceMaximum=50
    SourceMinimum=170
    TargetMinimum=0
    bClamp=true
    Value=/Script/RigVM.RigVMFunction_MathFloatAbs(Value=&RigVMFunction_MathQuaternionToEuler(RotationOrder=ZYX, Value=@n23.Transform.Rotation).Result.Z) {is=struct RigVMUnitNode, template="Absolute::Execute(in Value,out Result)"}.Result
   )
   n25 = &RigVMFunction_MathQuaternionToEuler(RotationOrder=ZYX, Value=&RigUnit_GetInitialBoneTransform(Bone=deform_pectoral_fr_l, Space=LocalSpace).Transform.Rotation)
   > &RigUnit_ModifyBoneTransforms(
    BoneToModify=((Bone="deform_pectoral_fr_l",Transform=(Rotation=(X=0,Y=0,Z=0,W=1),Translation=(X=4.877,Y=2.398,Z=1.048),Scale3D=(X=1,Y=1,Z=1))))
    Mode=OverrideLocal
    WeightMaximum=1
    Weight=@n24.Result
    BoneToModify.0.Transform.Rotation=&RigVMFunction_MathQuaternionFromEuler(Euler=&RigVMFunction_MathVectorAdd(B=(X=0,Y=0,Z=0), A=@n25.Result) {template="Add::Execute(in A,in B,out Result)"}.Result).Result
   )
   n26 = &RigVMFunction_MathQuaternionToEuler(RotationOrder=ZYX, Value=&RigUnit_GetInitialBoneTransform(Bone=deform_deltoid_fr_l, Space=LocalSpace).Transform.Rotation)
   > &RigUnit_ModifyBoneTransforms(
    BoneToModify=((Bone="deform_deltoid_fr_l",Transform=(Rotation=(X=0,Y=0,Z=0,W=1),Translation=(X=-4.337,Y=2.052,Z=0.026),Scale3D=(X=1,Y=1,Z=1))))
    Mode=OverrideLocal
    WeightMaximum=1
    WeightMinimum=0
    Weight=@n24.Result
    BoneToModify.0.Transform.Rotation=&RigVMFunction_MathQuaternionFromEuler(Euler=&RigVMFunction_MathVectorAdd(B=(X=0,Y=0,Z=0), A=@n26.Result) {template="Add::Execute(in A,in B,out Result)"}.Result).Result
   )
   n27 = &RigUnit_GetRelativeBoneTransform(Bone=upperarm_r, Space=spine_05)
   n28 = &RigVMFunction_MathFloatRemap(
    SourceMaximum=123
    SourceMinimum=30
    TargetMinimum=0
    bClamp=true
    Value=&RigVMFunction_MathFloatAbs(Value=&RigVMFunction_MathQuaternionToEuler(RotationOrder=ZYX, Value=@n27.Transform.Rotation).Result.Z) {template="Absolute::Execute(in Value,out Result)"}.Result
   )
   RigUnit_GetInitialBoneTransform = &RigUnit_GetInitialBoneTransform(Bone=deform_pectoral_fr_r, Space=LocalSpace)
   n29 = &RigVMFunction_MathQuaternionToEuler(RotationOrder=ZYX, Value=@RigUnit_GetInitialBoneTransform.Transform.Rotation)
   > &RigUnit_ModifyBoneTransforms(
    BoneToModify=((Bone="deform_pectoral_fr_r",Transform=(Rotation=(X=0,Y=0,Z=0,W=1),Translation=(X=-5.561,Y=-1.107,Z=-5.161),Scale3D=(X=1,Y=1,Z=1))))
    Mode=OverrideLocal
    WeightMaximum=1
    Weight=@n28.Result
    BoneToModify.0.Transform.Rotation=&RigVMFunction_MathQuaternionFromEuler(Euler=&RigVMFunction_MathVectorAdd(B=(X=0,Y=-17.297,Z=0), A=@n29.Result) {template="Add::Execute(in A,in B,out Result)"}.Result).Result
   )
   n30 = &RigVMFunction_MathQuaternionToEuler(RotationOrder=ZYX, Value=&RigUnit_GetInitialBoneTransform(Bone=deform_deltoid_fr_r, Space=LocalSpace).Transform.Rotation)
   > &RigUnit_ModifyBoneTransforms(
    BoneToModify=((Bone="deform_deltoid_fr_r",Transform=(Rotation=(X=0,Y=0,Z=0,W=1),Translation=(X=-0.063,Y=-3.462,Z=0.39),Scale3D=(X=1,Y=1,Z=1))))
    Mode=OverrideLocal
    WeightMaximum=1
    WeightMinimum=0
    Weight=@n28.Result
    BoneToModify.0.Transform.Rotation=&RigVMFunction_MathQuaternionFromEuler(Euler=&RigVMFunction_MathVectorAdd(B=(X=0,Y=28.666,Z=0), A=@n30.Result) {template="Add::Execute(in A,in B,out Result)"}.Result).Result
   )
   n31 = &RigUnit_GetRelativeBoneTransform(Bone=thigh_l, Space=pelvis)
   n32 = &RigVMFunction_MathFloatRemap(
    SourceMaximum=100
    SourceMinimum=2
    TargetMinimum=0
    bClamp=true
    Value=&RigVMFunction_MathFloatAbs(Value=&RigVMFunction_MathQuaternionToEuler(RotationOrder=ZYX, Value=@n31.Transform.Rotation).Result.Z) {template="Absolute::Execute(in Value,out Result)"}.Result
   )
   n33 = &RigVMFunction_MathQuaternionToEuler(RotationOrder=ZYX, Value=&RigUnit_GetInitialBoneTransform(Bone=deform_thigh_fr_l, Space=LocalSpace).Transform.Rotation)
   > &RigUnit_ModifyBoneTransforms(
    BoneToModify=((Bone="deform_thigh_fr_l",Transform=(Rotation=(X=0,Y=0,Z=0,W=1),Translation=(X=18,Y=12,Z=0),Scale3D=(X=1,Y=1,Z=1))))
    Mode=OverrideLocal
    WeightMaximum=1
    Weight=@n32.Result
    BoneToModify.0.Transform.Rotation=&RigVMFunction_MathQuaternionFromEuler(Euler=&RigVMFunction_MathVectorAdd(B=(X=0,Y=0,Z=45), A=@n33.Result) {template="Add::Execute(in A,in B,out Result)"}.Result).Result
   )
   n34 = &RigVMFunction_MathQuaternionToEuler(RotationOrder=ZYX, Value=&RigUnit_GetInitialBoneTransform(Bone=deform_thigh_bk_l, Space=LocalSpace).Transform.Rotation)
   > &RigUnit_ModifyBoneTransforms(
    BoneToModify=((Bone="deform_thigh_bk_l",Transform=(Rotation=(X=0,Y=0,Z=0,W=1),Translation=(X=25,Y=-5,Z=8),Scale3D=(X=1,Y=1,Z=1))))
    Mode=OverrideLocal
    WeightMaximum=1
    WeightMinimum=0
    Weight=@n32.Result
    BoneToModify.0.Transform.Rotation=&RigVMFunction_MathQuaternionFromEuler(Euler=&RigVMFunction_MathVectorAdd(B=(X=0,Y=0,Z=23.3), A=@n34.Result) {template="Add::Execute(in A,in B,out Result)"}.Result).Result
   )
   n35 = &RigUnit_GetRelativeBoneTransform(Bone=thigh_r, Space=pelvis)
   n36 = &RigVMFunction_MathFloatRemap(
    SourceMaximum=77
    SourceMinimum=177
    TargetMinimum=0
    bClamp=true
    Value=&RigVMFunction_MathFloatAbs(Value=&RigVMFunction_MathQuaternionToEuler(RotationOrder=ZYX, Value=@n35.Transform.Rotation).Result.Z) {template="Absolute::Execute(in Value,out Result)"}.Result
   )
   n37 = &RigVMFunction_MathQuaternionToEuler(RotationOrder=ZYX, Value=&RigUnit_GetInitialBoneTransform(Bone=deform_thigh_fr_r, Space=LocalSpace).Transform.Rotation)
   > &RigUnit_ModifyBoneTransforms(
    BoneToModify=((Bone="deform_thigh_fr_r",Transform=(Rotation=(X=0,Y=0,Z=0,W=1),Translation=(X=-18,Y=-12,Z=0),Scale3D=(X=1,Y=1,Z=1))))
    Mode=OverrideLocal
    WeightMaximum=1
    Weight=@n36.Result
    BoneToModify.0.Transform.Rotation=&RigVMFunction_MathQuaternionFromEuler(Euler=&RigVMFunction_MathVectorAdd(B=(X=0,Y=0,Z=45), A=@n37.Result) {template="Add::Execute(in A,in B,out Result)"}.Result).Result
   )
   n38 = &RigVMFunction_MathQuaternionToEuler(RotationOrder=ZYX, Value=&RigUnit_GetInitialBoneTransform(Bone=deform_thigh_bk_r, Space=LocalSpace).Transform.Rotation)
   > &RigUnit_ModifyBoneTransforms(
    BoneToModify=((Bone="deform_thigh_bk_r",Transform=(Rotation=(X=0,Y=0,Z=0,W=1),Translation=(X=-30,Y=4.5,Z=0.327),Scale3D=(X=1,Y=1,Z=1))))
    Mode=OverrideLocal
    WeightMaximum=1
    WeightMinimum=0
    Weight=@n36.Result
    BoneToModify.0.Transform.Rotation=&RigVMFunction_MathQuaternionFromEuler(Euler=&RigVMFunction_MathVectorAdd(B=(X=0,Y=0,Z=24), A=@n38.Result) {template="Add::Execute(in A,in B,out Result)"}.Result).Result
   )
  ]}
]
hierarchy [
 root {loc=<0 0 0>, rot=<0 0 0 1>}
 pelvis {parent=root, loc=<0 -0.5301 98.0571>, rot=<0.0127 0.707 -0.0127 -0.707>}
 spine_01 {parent=pelvis, loc=<6.1208 -0.9328 0>, rot=<0 0 0.034 -0.9994>}
 spine_02 {parent=spine_01, loc=<6.5996 -0.2095 0>, rot=<0 0 0.0238 -0.9997>}
 spine_03 {parent=spine_02, loc=<6.4914 -0.1237 0>, rot=<0 0 0.0149 -0.9999>}
 spine_04 {parent=spine_03, loc=<6.9598 0.0794 0>, rot=<0 0 -0.0132 -0.9999>}
 spine_05 {parent=spine_04, loc=<6.9987 0 0>, rot=<0 0 -0.0015 -1>}
 clavicle_l {parent=spine_05, loc=<5.4831 -0.1053 -3.2946>, rot=<0.0465 -0.7289 -0.009 -0.6829>}
 deform_pectoral_fr_l {parent=clavicle_l, loc=<6.2796 0 0>, rot=<0 0 0 1>}
 clavicle_r {parent=spine_05, loc=<5.4831 -0.1053 3.2946>, rot=<0.7289 0.0465 -0.6829 0.009>}
 deform_pectoral_fr_r {parent=clavicle_r, loc=<-6.28 0 0>, rot=<0 0 0 1>}
 neck_01 {parent=spine_05, loc=<11.3138 0.1529 0>, rot=<0 0 -0.2003 -0.9797>}
 neck_02 {parent=neck_01, loc=<4.5627 0 0>, rot=<0 0 0.0504 -0.9987>}
 head {parent=neck_02, loc=<5.3914 0 0>, rot=<0 0 0.115 -0.9934>}
 faceAttach {parent=head, loc=<0.6201 0.9296 0>, rot=<0.0082 -0.7071 -0.0082 -0.7071>, scale=<0.89 0.89 0.89>}
 C_jaw {parent=faceAttach, loc=<-0.091 0.7508 -3.585>, rot=<0 0 0 1>}
 Dyn_Master_Hair {parent=head, loc=<11.1235 6.1693 -3.2045>, rot=<-0.0002 0.005 0.9995 -0.0317>}
 Dyn_Hair5 {parent=Dyn_Master_Hair, loc=<-1.651 -3.0129 2.4474>, rot=<0.0254 0.7078 0.0146 -0.7058>}
 Dyn_Hair1 {parent=Dyn_Master_Hair, loc=<0.8225 0.642 -1.2158>, rot=<-0.6569 0.0197 0.0203 -0.7534>}
 Dyn_Hair4 {parent=Dyn_Master_Hair, loc=<0.5081 -2.8586 2.4496>, rot=<0.0186 0.4118 0.0226 -0.9108>}
 Dyn_Hair2 {parent=Dyn_Master_Hair, loc=<0.8913 -0.5438 -0.8438>, rot=<-0.5597 0.0853 0.1186 -0.8157>}
 Dyn_Hair3 {parent=Dyn_Master_Hair, loc=<1.57 -2.1683 0.5711>, rot=<-0.2695 0.3757 -0.0987 -0.8812>}
 Dyn_Hair6 {parent=Dyn_Master_Hair, loc=<-0.5057 -0.849 7.1031>, rot=<0.0255 0.7084 0.0145 -0.7052>}
 Dyn_Hair7 {parent=Dyn_Master_Hair, loc=<0.8019 -2.0511 5.7346>, rot=<0.0077 0.0013 0.0283 -0.9996>}
 Dyn_Hair8 {parent=Dyn_Master_Hair, loc=<1.5353 -0.5816 7.7328>, rot=<0.0255 0.7113 0.0144 -0.7023>}
 Dyn_Master_Antenna_l {parent=head, loc=<12.9838 2.0403 5.1143>, rot=<-0.0066 0.2022 -0.0321 -0.9788>}
 Dyn_Antenna_l1 {parent=Dyn_Master_Antenna_l, loc=<0.5391 0 0>, rot=<0 0 0 1>}
 Dyn_Antenna_l2 {parent=Dyn_Antenna_l1, loc=<2.9096 0 0>, rot=<0.0001 -0.0157 -0.0069 -0.9999>}
 Dyn_Antenna_l3 {parent=Dyn_Antenna_l2, loc=<5.3545 0 0>, rot=<0.0006 -0.0217 -0.0284 -0.9994>}
 Dyn_Antenna_l4 {parent=Dyn_Antenna_l3, loc=<5.8121 0 -0.017>, rot=<0 -0.1474 0 -0.9891>}
 Dyn_Master_Antenna_r {parent=head, loc=<12.9618 2.0389 -5.1318>, rot=<0.0064 -0.1947 -0.0322 -0.9803>}
 Dyn_Antenna_r1 {parent=Dyn_Master_Antenna_r, loc=<0.5391 0 0>, rot=<0 0 0 1>}
 Dyn_Antenna_r2 {parent=Dyn_Antenna_r1, loc=<2.9096 0 0>, rot=<0 0.0056 -0.0069 -1>}
 Dyn_Antenna_r3 {parent=Dyn_Antenna_r2, loc=<5.3545 0 0>, rot=<-0.0004 0.0157 -0.0284 -0.9995>}
 Dyn_Antenna_r4 {parent=Dyn_Antenna_r3, loc=<5.8121 0 -0.017>, rot=<0 0.1592 0 -0.9872>}
 Dyn_Master_Flap {parent=spine_05, loc=<5.8108 -8.2044 0>, rot=<0.0395 -0.9992 0 0>}
 Dyn_Master_Flap_l {parent=Dyn_Master_Flap, loc=<0 0 -4.9383>, rot=<0 0 0 -1>}
 Dyn_Flap_l1 {parent=Dyn_Master_Flap_l, loc=<0.2279 0 0>, rot=<0 0 0 -1>}
 Dyn_Flap_l2 {parent=Dyn_Flap_l1, loc=<3.8628 0 0>, rot=<0 0 0 -1>}
 Dyn_Flap_l3 {parent=Dyn_Flap_l2, loc=<6.2682 0 0>, rot=<0 0 -0.0076 -1>}
 Dyn_Flap_l4 {parent=Dyn_Flap_l3, loc=<4.4726 0 0>, rot=<0 0 0 1>}
 Dyn_Master_Flap_r {parent=Dyn_Master_Flap, loc=<0 0 4.9189>, rot=<0 0 0 -1>}
 Dyn_Flap_r1 {parent=Dyn_Master_Flap_r, loc=<0.2279 0 0>, rot=<0 0 0 -1>}
 Dyn_Flap_r2 {parent=Dyn_Flap_r1, loc=<3.8628 0 0>, rot=<0 0 0 -1>}
 Dyn_Flap_r3 {parent=Dyn_Flap_r2, loc=<6.2682 0 0>, rot=<0 0 -0.0076 -1>}
 Dyn_Flap_r4 {parent=Dyn_Flap_r3, loc=<4.4726 0 0>, rot=<0 0 0 1>}
 attach {parent=root, loc=<0 0 0>, rot=<0 0 0 1>}
 dyn_root_simspace {parent=root, loc=<0 0 -2.1558>, rot=<0 0 0.401 -0.9161>}
 dyn_root_slope {parent=root, loc=<0 0.0001 -2.1567>, rot=<0 0 0 1>}
]
